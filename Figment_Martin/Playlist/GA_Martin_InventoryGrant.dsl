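; Gameplay ability for the Figment Martin playlist that grants loot-table items to player pawns carrying the Hero tag.
; On activation it validates the avatar, checks the tag, and gets the inventory owner interface from the pawn's controller.
; It then picks drops from a regular weapon tier and a mobility tier in sequence, giving each item definition to that inventory.
; The graph never calls EndAbility, and its end event is empty.
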(event Ability|EventActivateAbility
  (bind _returnvalue (Ability|GetAvatarActorfromActorInfo))
  (Utilities|IsValid _returnvalue
    (:Is%20Valid
      (bind _asfort_player_pawn (Utilities|Casting|CastToFortPlayerPawn _returnvalue))
      (if (GameplayTags|TagContainer|CompareTagAssetInterfacetoOtherTags (Variables|Default|GetTag_Hero) _asfort_player_pawn)
        (bind _asfort_inventory_owner_interface (Utilities|Casting|CastToFortInventoryOwnerInterface (Fortnite|GetControllerfromActor _returnvalue)))
        (Utilities|FlowControl|Sequence
          (:then_0
            ; Get Loot Drop From Regular Weapon Pool
            (bind _outloottodrop (Loot|PickLootDrops (Variables|Default|GetLootTierName) (Loot|GetLootLevel) -1 "(GameplayTags=)"))
            (for _array_element _outloottodrop
              (bind _asfort_world_item_definition (Utilities|Casting|CastToFortWorldItemDefinition (Class|FortItemFunctionLibrary|BreakFortItemEntry _array_element)))
              (Inventory|GiveItemtoInventoryOwner _asfort_inventory_owner_interface _asfort_world_item_definition)))
          (:then_1
            ; Get Loot Drop from Mobility Weapon Pool
            (bind _outloottodrop_2 (Loot|PickLootDrops (Variables|Default|GetLootTierName_Mobility) (Loot|GetLootLevel) -1 "(GameplayTags=)"))
            (for _array_element_2 _outloottodrop_2
              (bind _asfort_world_item_definition_2 (Utilities|Casting|CastToFortWorldItemDefinition (Class|FortItemFunctionLibrary|BreakFortItemEntry _array_element_2)))
              (Inventory|GiveItemtoInventoryOwner _asfort_inventory_owner_interface _asfort_world_item_definition_2))))))))

(event Ability|EventOnEndAbility
  (params (bWasCancelled :type Boolean)))
