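; GA_MatDrip is a Figment_Martin playlist gameplay ability that runs item award timers, apparently to drip building materials to players.
; On activation it runs generic setup, begins waiting for the Milo player tag on the pawn, and starts the timers.
; When that tag is added, it reads the Milo Building row from the Figment_Martin data table and stops the timers if the value is below one.
; Ending the ability also stops them.
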(event Ability|EventActivateAbility
  (bind (_event_data _fort_player_pawn) (Setup|SetupGenericAbility true)
    (:Output
      (Ability|Tasks|WaitGameplayTagAdd (Variables|Default|GetMiloPlayerTag) _fort_player_pawn
        (:then
          _)
        (:Added
          (bind _outxy (DataTable|EvaluateCurveTableRow "/Figment_Martin/Data/Figment_Martin_Data.Figment_Martin_Data" "Default.Ashton.Milo.Building")
            (:RowFound
              (if (not (>= _outxy 1.0))
                (CallFunction|StopItemAwardTimers))))))
      (CallFunction|StartItemAwardTimers))))

(event Ability|EventOnEndAbility
  (params (bWasCancelled :type Boolean))
  (CallFunction|StopItemAwardTimers))
